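; Creative Item Granter device that gives players or teams items registered by dropping pickups onto it outside gameplay.
; Trigger events cycle the item index or build grant requests, held pending until minigame start.
; A server-side queue processes requests on a short timer, async-loading items, applying conditions and behaviors, then granting or dropping them and optionally equipping.
; An optional repeat timer uses saved per-player timestamps to multiply grants for time spent offline.
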(event Custom|Process%20Grant%20Request%20Queue
  (bind (_pickupinstigatorhandle _itemvariantguid) (Utilities|Struct|BreakMinigameSpawnerSpawnParams (Variables|Transient|GetProcessingGrantRequestSpawnParams)))
  (bind (_pickupinstigatorhandle_2 _itemvariantguid_2) (Utilities|Struct|BreakMinigameSpawnerSpawnParams (Variables|Transient|GetProcessingGrantRequestSpawnParams)))
  (bind (_processinggrantrequest_controllers_38_822ad16d47dcab66ef0d7c9ce6898856 _processinggrantrequest_grantitemindex_21_b2c0fe6d40d1128d9ea9c192f324d978) (Utilities|Struct|BreakSCreativeItemGranterGrantRequest (Variables|GetProcessingGrantRequest_Controllers_38_822AD16D47DCAB66EF0D7C9CE6898856)))
  (bind (_best_valid_item_index _isvalid) (CallFunction|GetBestValidItemIndex _processinggrantrequest_grantitemindex_21_b2c0fe6d40d1128d9ea9c192f324d978))
  (bind _validitemindexes (Variables|Transient|GetValidItemIndexes))
  (bind _output (Utilities|Array|Get(aref) _validitemindexes _best_valid_item_index))
  (bind (_processinggrantrequest_controllers_38_822ad16d47dcab66ef0d7c9ce6898856_2 _processinggrantrequest_grantitemindex_21_b2c0fe6d40d1128d9ea9c192f324d978_2 _processinggrantrequest_countoverride_41_d8a66e564fea2a9fded634a7eddc274c _processinggrantrequest_executeitem_8_e7c22d144c7730e28a63608db73c7276 _processinggrantrequest_removeitem_20_8b360fe344295a37c03d84bfac65befd _processinggrantrequest_grantbehavior_32_af1cd62340c36f7d698dd59f7c7506f9) (Utilities|Struct|BreakSCreativeItemGranterGrantRequest (Variables|GetProcessingGrantRequest_Controllers_38_822AD16D47DCAB66EF0D7C9CE6898856)))
  (bind (_processinggrantrequest_controllers_38_822ad16d47dcab66ef0d7c9ce6898856_3 _processinggrantrequest_grantitemindex_21_b2c0fe6d40d1128d9ea9c192f324d978_3 _processinggrantrequest_countoverride_41_d8a66e564fea2a9fded634a7eddc274c_2 _processinggrantrequest_executeitem_8_e7c22d144c7730e28a63608db73c7276_2) (Utilities|Struct|BreakSCreativeItemGranterGrantRequest (Variables|GetProcessingGrantRequest_Controllers_38_822AD16D47DCAB66EF0D7C9CE6898856)))
  (bind _processinggrantrequest_controllers_38_822ad16d47dcab66ef0d7c9ce6898856_4 (Utilities|Struct|BreakSCreativeItemGranterGrantRequest (Variables|GetProcessingGrantRequest_Controllers_38_822AD16D47DCAB66EF0D7C9CE6898856)))
  (bind (_processinggrantrequest_controllers_38_822ad16d47dcab66ef0d7c9ce6898856_5 _processinggrantrequest_grantitemindex_21_b2c0fe6d40d1128d9ea9c192f324d978_4 _processinggrantrequest_countoverride_41_d8a66e564fea2a9fded634a7eddc274c_3 _processinggrantrequest_executeitem_8_e7c22d144c7730e28a63608db73c7276_3 _processinggrantrequest_removeitem_20_8b360fe344295a37c03d84bfac65befd_2 _processinggrantrequest_grantbehavior_32_af1cd62340c36f7d698dd59f7c7506f9_2 _processinggrantrequest_grantcondition_34_0e6824e546464c43b5988bad6d7f9c9e) (Utilities|Struct|BreakSCreativeItemGranterGrantRequest (Variables|GetProcessingGrantRequest_Controllers_38_822AD16D47DCAB66EF0D7C9CE6898856)))
  (bind _processinggrantrequestinventoryowner (Variables|Transient|GetProcessingGrantRequestInventoryOwner))
  (bind _processinggrantrequestinventoryowner_2 (Variables|Transient|GetProcessingGrantRequestInventoryOwner))
  (bind _processinggrantrequestinventoryowner_3 (Variables|Transient|GetProcessingGrantRequestInventoryOwner))
  (bind (_processinggrantrequest_controllers_38_822ad16d47dcab66ef0d7c9ce6898856_6 _processinggrantrequest_grantitemindex_21_b2c0fe6d40d1128d9ea9c192f324d978_5 _processinggrantrequest_countoverride_41_d8a66e564fea2a9fded634a7eddc274c_4 _processinggrantrequest_executeitem_8_e7c22d144c7730e28a63608db73c7276_4 _processinggrantrequest_removeitem_20_8b360fe344295a37c03d84bfac65befd_3 _processinggrantrequest_grantbehavior_32_af1cd62340c36f7d698dd59f7c7506f9_3 _processinggrantrequest_grantcondition_34_0e6824e546464c43b5988bad6d7f9c9e_2 _processinggrantrequest_lastitem_36_f36492d5444f9101a9ea0bab57a98f7e) (Utilities|Struct|BreakSCreativeItemGranterGrantRequest (Variables|GetProcessingGrantRequest_Controllers_38_822AD16D47DCAB66EF0D7C9CE6898856)))
  (Networking|SwitchHasAuthority
    (:Authority
      ;[0 1 0] Ensure the mini game is in progress
      (bind _fortminigameprogress (Variables|Default|GetFortMinigameProgress)
        (:then
          (bind _currentminigame (Variables|/Script/FortniteGame.FortMinigameLogicComponent|GetCurrentMinigame _fortminigameprogress)
            (:then
              (if (Minigame|IsinProgress _currentminigame)
                (if (Math|Boolean|ANDBoolean (Utilities|IsArrayNonEmpty (Variables|Transient|GetGrantRequestQueue)) (not (Variables|Transient|GetbProcessingGrantRequests)) (Variables|Transient|GetbHasCalledOnMinigameStarted))
                  (Variables|Transient|SetbProcessingGrantRequests true)
                  ;[1 0 0] Grab and remove the first one off the queue
                  (Variables|Transient|SetProcessingGrantRequest (Utilities|Array|Get(aref) (Variables|Transient|GetGrantRequestQueue)))
                  (Utilities|Array|RemoveIndex (Variables|Transient|GetGrantRequestQueue))
                  ;[1 0.25 0] Evaluate the best valid item index and then load the soft object pointer
                  (if (and _isvalid (Utilities|Array|IsValidIndex (Variables|Transient|GetValidItemIndexes) _best_valid_item_index))
                    (if (!= _output -1)
                      (Variables|Transient|SetProcessingGrantRequestItemIndex _output)
                      (bind _output_get (Variables|Transient|SetProcessingGrantRequestSpawnParams (ItemData|MakeSpawnParams (Variables|Default|GetPickupItemList) _output (Transformation|GetActorTransform))))
                      (if (Utilities|IsValidSoftObjectReference (Utilities|Struct|BreakMinigameSpawnerSpawnParams _output_get))
                        (bind _object (Utilities|AsyncLoadAsset (Utilities|Struct|BreakMinigameSpawnerSpawnParams _output_get))
                          (:Completed
                            (Utilities|FlowControl|Sequence
                              (:then_0
                                (bind _asfort_world_item_definition (Utilities|Casting|CastToFortWorldItemDefinition _object))
                                (Variables|Transient|SetProcessingGrantRequestWorldItemDefinition _asfort_world_item_definition)
                                ;[0 0 1] Loop through the players for this grant request and ensure the player controller is valid
                                (for _array_element _processinggrantrequest_controllers_38_822ad16d47dcab66ef0d7c9ce6898856_4
                                  (Variables|Transient|SetProcessingGrantRequestController 0)
                                  (Variables|Transient|SetProcessingGrantRequestInventoryOwner 0)
                                  (if (Utilities|IsValid _array_element :!class /Script/Engine.KismetSystemLibrary)
                                    (bind _playerstate (Variables|/Script/Engine.Controller|GetPlayerState _array_element)
                                      (:then
                                        (bind _asfort_player_state (Utilities|Casting|CastToFortPlayerState _playerstate))
                                        (if (Minigame|IsPlayerinMinigame (Variables|/Script/FortniteGame.FortMinigameLogicComponent|GetCurrentMinigame (Variables|Default|GetFortMinigameProgress)) _asfort_player_state false)
                                          (bind _output_get_2 (Variables|Transient|SetProcessingGrantRequestController _array_element))
                                          (bind _asfort_inventory_owner_interface (Utilities|Casting|CastToFortInventoryOwnerInterface _output_get_2))
                                          (Variables|Transient|SetProcessingGrantRequestInventoryOwner _asfort_inventory_owner_interface)
                                          ;[0.25 0 1] Check grant conditions to ensure we should give the item to the player
                                          (switch Utilities|FlowControl|Switch|SwitchonE_Device_ItemGranter_GrantCondition _processinggrantrequest_grantcondition_34_0e6824e546464c43b5988bad6d7f9c9e
                                            (:NewEnumerator0
                                              _)
                                            (:NewEnumerator1
                                              (if (not (Inventory|DoesInventoryOwnerHaveItems _processinggrantrequestinventoryowner (Variables|Default|GetExcludedItemTypes)))
                                                (Variables|Default|SetbOnlyIfInitiallyEmpty true)))
                                            (:NewEnumerator2
                                              (if (> (Inventory|GetInventorySpace _processinggrantrequestinventoryowner) 0)
                                                _))
                                            (:NewEnumerator19
                                              (if (Variables|Default|GetbOnlyIfInitiallyEmpty)
                                                _))
                                            (:NewEnumerator20
                                              _))
                                          ;[1 1 0] Before granting the item, perform a behavior
                                          (switch Utilities|FlowControl|Switch|SwitchonE_Device_ItemGranter_GrantBehavior _processinggrantrequest_grantbehavior_32_af1cd62340c36f7d698dd59f7c7506f9
                                            (:NewEnumerator0
                                              (Inventory|ClearDroppableItems _processinggrantrequestinventoryowner_2))
                                            (:NewEnumerator1
                                              (Inventory|RemoveNonResourceDroppableItems _processinggrantrequestinventoryowner_2))
                                            (:NewEnumerator2
                                              (Inventory|ReduceDroppableResources _processinggrantrequestinventoryowner_2))
                                            (:NewEnumerator3
                                              _))
                                          ;[1 1 1] Plug in Multiplier check by getting the saved Date Data here.  Need to Check if player is coming back vs if the device was disabled.
                                          (if (== (Variables|ToyProperties|GetGrantWhileOffline) 1)
                                            (Default|CalculateItemstoGrantOverTime)
                                            (else
                                              (Variables|Transient|SetItemGrantMultiplier 1)))
                                          ;[1 1 1] Audio
                                          (if (> (Variables|UserOptions|GetPlayAudioKnob) 0)
                                            (Creative|GameplayEffects|SimpleApplyGameplayEffectToActor (Variables|Default|GetItemGrantedGameplayEffect) (Variables|Transient|GetProcessingGrantRequestController)))
                                          (if (CallFunction|ShouldSpawnedItemDroptoGround)
                                            ;[0.02416 0.3467 0.2051] Drop item to ground
                                            (Inventory|SpawnPickupInWorldWithClassAndItemEntry (Creative|ItemData|CreateItemVariantEntry (Variables|Transient|GetProcessingGrantRequestWorldItemDefinition) (CallFunction|CalculateItemNumber) -1 _itemvariantguid _pickupinstigatorhandle) "/Game/Creative/Items/Gameplay/FortCreativePickup/Fort_Pickup_Creative.Fort_Pickup_Creative_C" (Transformation|GetActorLocation (Pawn|GetControlledPawn (Variables|Transient|GetProcessingGrantRequestController))) "0.000000,0.000000,1.000000" 0 true true false "Tossed" "ItemSpawner" (Utilities|Select :Index (Variables|UserOptions|GetOwnershipOfDroppedItem) :NewEnumerator2 (Utilities|Casting|CastToFortPlayerControllerAthena (Variables|Transient|GetProcessingGrantRequestController)) :NewEnumerator1 (Utilities|Casting|CastToFortPlayerControllerAthena (Variables|Transient|GetTriggerInstigator))))
                                            (else
                                              ;[0.02416 0.3467 0.2051] Grant the item
                                              (Inventory|GiveItemtoInventoryOwner _processinggrantrequestinventoryowner_3 (Variables|Transient|GetProcessingGrantRequestWorldItemDefinition) _itemvariantguid_2 (CallFunction|CalculateItemNumber) true -1 _pickupinstigatorhandle_2)
                                              (CallFunction|MaybeGiveAmmotoInventoryOwner _processinggrantrequestinventoryowner_3 (Variables|Transient|GetProcessingGrantRequestWorldItemDefinition))))
                                          (CallFunction|SendwhenItemGrantedTriggerMessage (Variables|Transient|GetProcessingGrantRequestController))
                                          ;[0.3539 0.6908 1] Try Equip Item
                                          (if (and _processinggrantrequest_executeitem_8_e7c22d144c7730e28a63608db73c7276_2 (CallFunction|IsPlayerAllowedtoEquipProcessingItem))
                                            (bind _asfort_player_controller (Utilities|Casting|CastToFortPlayerController (Variables|Transient|GetProcessingGrantRequestController)))
                                            (Inventory|ExecuteInventoryItemDefinition _asfort_player_controller (Variables|Transient|GetProcessingGrantRequestWorldItemDefinition) (Utilities|Struct|BreakMinigameSpawnerSpawnParams (Variables|Transient|GetProcessingGrantRequestSpawnParams)) 0.0 true true))))
                                      (:else)))))
                              (:then_1
                                _))))
                        (else
                          _))
                      ;[1 0 1] Clean up the process request and do any post grant request actions
                      (Utilities|FlowControl|Sequence
                        (:then_0
                          (if _processinggrantrequest_removeitem_20_8b360fe344295a37c03d84bfac65befd_3
                            (Utilities|Array|SetArrayElem (Variables|Transient|GetValidItemIndexes) (Variables|Transient|GetProcessingGrantRequestItemIndex) -1)))
                        (:then_1
                          (if _processinggrantrequest_lastitem_36_f36492d5444f9101a9ea0bab57a98f7e
                            (Variables|Default|SetbOnlyIfInitiallyEmpty false)))
                        (:then_2
                          (Variables|Transient|SetbProcessingGrantRequests false)
                          (Variables|Transient|SetProcessingGrantRequest))
                        (:then_3
                          (if (Utilities|IsArrayNonEmpty (Variables|Transient|GetGrantRequestQueue))
                            (bind _returnvalue (Utilities|Time|SetTimerbyEvent (EventDispatchers|CreateEvent :!event Custom|Process%20Grant%20Request%20Queue) 0.01))
                            (Variables|Transient|SetProcessingGrantRequestTimerHandle _returnvalue))))
                      (else
                        _))
                    (else
                      _)))))
            (:else)))
        (:else)))))

(event Custom|EnableFromEvent
  (params (Instigator :type Controller%20Object%20Reference))
  ;[0.6921 0 1] Triggers & Receivers
  (Networking|SwitchHasAuthority
    (:Authority
      (if (CallFunction|IsinProgress)
        (Variables|Transient|SetbEnabled true)))))

(event Custom|Reset
  ;[0 0.6807 1] Reset Grant Queue
  (Utilities|Array|Clear (Variables|Transient|GetGrantRequestQueue))
  (Variables|Transient|SetbProcessingGrantRequests false)
  (Variables|Transient|SetProcessingGrantRequest)
  (Utilities|Time|ClearandInvalidateTimerbyHandle (Variables|Transient|GetProcessingGrantRequestTimerHandle))
  (Variables|Transient|SetProcessingGrantRequestController 0)
  (Variables|Transient|SetProcessingGrantRequestInventoryOwner 0)
  (Variables|Transient|SetProcessingGrantRequestItemIndex -1)
  (Variables|Transient|SetProcessingGrantRequestSpawnParams)
  (Variables|Transient|SetProcessingGrantRequestWorldItemDefinition 0)
  (Variables|Transient|SetItemGrantMultiplier 1))

(event Custom|DisableFromEvent
  (params (Instigator :type Controller%20Object%20Reference))
  ;[0.6921 0 1] Triggers & Receivers
  (Networking|SwitchHasAuthority
    (:Authority
      (if (CallFunction|IsinProgress)
        (Variables|Transient|SetbEnabled false)))))

(event EventOnMinigameStarted
  ;[0 1 0.4804] Game Start
  (CallFunction|Reset)
  (Variables|Transient|SetbHasCalledOnMinigameStarted true)
  (if (Utilities|IsArrayNonEmpty (Variables|Transient|GetPendingGrantRequestQueue))
    (Variables|SetBy-RefVar (Variables|Transient|GetGrantRequestQueue) (Variables|Transient|GetPendingGrantRequestQueue))
    (Utilities|Array|Clear (Variables|Transient|GetPendingGrantRequestQueue))
    (CallFunction|ProcessGrantRequestQueue)))

(event Custom|TimerGrantItem
  (bind (_teamindex _returnvalue) (Team|GetActorPvPTeamIndex (Variables|Transient|GetTriggerInstigator)))
  ;[0.6921 0 1] Triggers & Receivers
  (Networking|SwitchHasAuthority
    (:Authority
      (if (Math|Boolean|ANDBoolean (Variables|Transient|GetbEnabled) (CallFunction|IsinProgress) (> (Utilities|Array|Length (Variables|Transient|GetValidItemIndexes)) 0))
        (if (!= (Variables|ToyProperties|GetGrantItemTimer) 0.0)
          (CallFunction|StartGrantItemTimer)
          (else
            _))
        (if (Knobs|Team|IsTeamTriggeringPlayer (Variables|ToyProperties|GetReceivingPlayers))
          (CallFunction|CreateGrantRequestforPlayers (Utilities|Array|MakeArray (Variables|Transient|GetTriggerInstigator)))
          (elif (Knobs|Team|IsTeamAll (Variables|ToyProperties|GetReceivingPlayers))
            (CallFunction|CreateGrantRequestforGroup 0 true)
            (else
              (CallFunction|CreateGrantRequestforGroup (select (and (Knobs|Team|IsTriggeringTeam (Variables|ToyProperties|GetReceivingPlayers)) _returnvalue) _teamindex (Math|Conversions|ToByte(Integer) (Knobs|Team|GetTeamValueasIndex (Variables|ToyProperties|GetReceivingPlayers))))))))))))

(event Custom|GrantItemFromEvent
  (params (Instigator :type Controller%20Object%20Reference))
  (bind (_teamindex _returnvalue) (Team|GetActorPvPTeamIndex (Variables|Transient|GetTriggerInstigator)))
  ;[0.6921 0 1] Triggers & Receivers
  (Variables|Transient|SetTriggerInstigator Instigator)
  (Variables|Default|SetEGrantProcessType "NewEnumerator0")
  (Networking|SwitchHasAuthority
    (:Authority
      (if (Math|Boolean|ANDBoolean (Variables|Transient|GetbEnabled) (CallFunction|IsinProgress) (> (Utilities|Array|Length (Variables|Transient|GetValidItemIndexes)) 0))
        (if (!= (Variables|ToyProperties|GetGrantItemTimer) 0.0)
          (CallFunction|StartGrantItemTimer)
          (else
            _))
        (if (Knobs|Team|IsTeamTriggeringPlayer (Variables|ToyProperties|GetReceivingPlayers))
          (CallFunction|CreateGrantRequestforPlayers (Utilities|Array|MakeArray (Variables|Transient|GetTriggerInstigator)))
          (elif (Knobs|Team|IsTeamAll (Variables|ToyProperties|GetReceivingPlayers))
            (CallFunction|CreateGrantRequestforGroup 0 true)
            (else
              (CallFunction|CreateGrantRequestforGroup (select (and (Knobs|Team|IsTriggeringTeam (Variables|ToyProperties|GetReceivingPlayers)) _returnvalue) _teamindex (Math|Conversions|ToByte(Integer) (Knobs|Team|GetTeamValueasIndex (Variables|ToyProperties|GetReceivingPlayers))))))))))))

(event MinigameStartedDelegate(FortMinigameProgress)
  (Utilities|FlowControl|Sequence
    (:then_0
      (Variables|Transient|SetbEnabled (Math|Conversions|ToBoolean(Integer) (Variables|ToyProperties|GetEnabledOnGameStart)))
      (Variables|Transient|SetGrantItemIndex 0))
    (:then_1
      (CallFunction|Restock))
    (:then_2
      (bind _returnvalue (Game|ConstructObjectfromClass (Variables|Default|GetPersistenceDataItemGranterClass) self))
      (Variables|Default|SetPersistenceDataItemGranter _returnvalue))))

(event Custom|Grant%20Item
  (bind (_teamindex _returnvalue) (Team|GetActorPvPTeamIndex (Variables|Transient|GetTriggerInstigator)))
  ;[0.6921 0 1] Triggers & Receivers
  (Networking|SwitchHasAuthority
    (:Authority
      (if (Math|Boolean|ANDBoolean (Variables|Transient|GetbEnabled) (CallFunction|IsinProgress) (> (Utilities|Array|Length (Variables|Transient|GetValidItemIndexes)) 0))
        (if (!= (Variables|ToyProperties|GetGrantItemTimer) 0.0)
          (CallFunction|StartGrantItemTimer)
          (else
            _))
        (if (Knobs|Team|IsTeamTriggeringPlayer (Variables|ToyProperties|GetReceivingPlayers))
          (CallFunction|CreateGrantRequestforPlayers (Utilities|Array|MakeArray (Variables|Transient|GetTriggerInstigator)))
          (elif (Knobs|Team|IsTeamAll (Variables|ToyProperties|GetReceivingPlayers))
            (CallFunction|CreateGrantRequestforGroup 0 true)
            (else
              (CallFunction|CreateGrantRequestforGroup (select (and (Knobs|Team|IsTriggeringTeam (Variables|ToyProperties|GetReceivingPlayers)) _returnvalue) _teamindex (Math|Conversions|ToByte(Integer) (Knobs|Team|GetTeamValueasIndex (Variables|ToyProperties|GetReceivingPlayers))))))))))))

(event EventOnMinigameEnded
  ;[1 0.1469 0] Game End
  (CallFunction|Reset)
  (Variables|Transient|SetbHasCalledOnMinigameStarted false)
  (Utilities|Array|Clear (Variables|Transient|GetPendingGrantRequestQueue))
  (CallFunction|ClearGrantItemTimer)
  (Utilities|Array|Clear (Variables|Transient|GetPlayersGrantedItemsThisSession)))

(event MinigameEndedDelegate(FortMinigameProgress)
  (Variables|Transient|SetbEnabled (Math|Conversions|ToBoolean(Integer) (Variables|ToyProperties|GetEnabledOnGameStart)))
  (Variables|Transient|SetGrantItemIndex 0))

(event EventDestroyed
  ;[1 0.08905 0] On Destroyed
  (if (Audio|Components|Audio|IsPlaying (Variables|Components|GetHolo%20Loop%20Cue))
    (Audio|Components|Audio|FadeOut (Variables|Components|GetHolo%20Loop%20Cue) 0.2 0.0 "Linear" :!class /Script/Engine.AudioComponent)))

(event Custom|TimerGrantPrevious
  ;[0.6921 0 1] Triggers & Receivers
  (Variables|Transient|SetGrantItemIndex (- (Variables|Transient|GetGrantItemIndex) 1))
  (Variables|Default|SetEGrantProcessType "NewEnumerator1")
  (switch Utilities|FlowControl|Switch|SwitchonE_Device_ItemGranter_CycleBehavior (Variables|ToyProperties|GetCycleBehavior)
    (:NewEnumerator0
      (Variables|Transient|SetGrantItemIndex (Math|Integer|Max(Integer) (Variables|Transient|GetGrantItemIndex))))
    (:NewEnumerator1
      (if (< (Variables|Transient|GetGrantItemIndex) 0)
        (Variables|Transient|SetGrantItemIndex (Utilities|Array|LastIndex (Variables|Transient|GetValidItemIndexes)))
        (else
          _))))
  (if (Math|Conversions|ToBoolean(Integer) (Variables|ToyProperties|GetGrantOnIndexChange))
    (CallFunction|GrantItem)))

(event Custom|CycleToPreviousItemFromEvent
  (params (Instigator :type Controller%20Object%20Reference))
  ;[0.6921 0 1] Triggers & Receivers
  (Variables|Transient|SetTriggerInstigator Instigator)
  (Variables|Transient|SetGrantItemIndex (- (Variables|Transient|GetGrantItemIndex) 1))
  (Variables|Default|SetEGrantProcessType "NewEnumerator1")
  (switch Utilities|FlowControl|Switch|SwitchonE_Device_ItemGranter_CycleBehavior (Variables|ToyProperties|GetCycleBehavior)
    (:NewEnumerator0
      (Variables|Transient|SetGrantItemIndex (Math|Integer|Max(Integer) (Variables|Transient|GetGrantItemIndex))))
    (:NewEnumerator1
      (if (< (Variables|Transient|GetGrantItemIndex) 0)
        (Variables|Transient|SetGrantItemIndex (Utilities|Array|LastIndex (Variables|Transient|GetValidItemIndexes)))
        (else
          _))))
  (if (Math|Conversions|ToBoolean(Integer) (Variables|ToyProperties|GetGrantOnIndexChange))
    (CallFunction|GrantItem)))

(event OnComponentHit(DropBox)
  (params (HitComponent :type Primitive%20Component%20Object%20Reference) (OtherActor :type Actor%20Object%20Reference) (OtherComp :type Primitive%20Component%20Object%20Reference) (NormalImpulse :type Vector :default "0, 0, 0") (Hit :type Hit%20Result%20Structure%20%28by%20ref%29))
  ;[0 0.4352 0.8594 0] Item Dropped on Device
  (Networking|SwitchHasAuthority
    (:Authority
      (if (not (CallFunction|IsinProgress))
        (bind _asfort_pickup_athena (Utilities|Casting|CastToFortPickupAthena OtherActor))
        (Utilities|IsPickupValid _asfort_pickup_athena (Variables|Default|GetInvalidPickupGameplayTagContainer)
          (:IsValid
            (if (and (Interaction|BlueprintCanInteract (Pickup|GetPawnWhoDroppedPickup _asfort_pickup_athena)) (not (Class|PickupItemListComponent|IsIteminQueue (Variables|Default|GetPickupItemList) _asfort_pickup_athena)))
              (bind _bitemadded (Class|PickupItemListComponent|AddItemtoQueue (Variables|Default|GetPickupItemList) _asfort_pickup_athena))
              (if _bitemadded
                (CallFunction|RPCMULTICASTAddItem _asfort_pickup_athena)))))))))

(event Custom|TimerGrantNext
  ;[0.6921 0 1] Triggers & Receivers
  (Variables|Transient|SetGrantItemIndex (+ (Variables|Transient|GetGrantItemIndex) 1))
  (Variables|Default|SetEGrantProcessType "NewEnumerator19")
  (switch Utilities|FlowControl|Switch|SwitchonE_Device_ItemGranter_CycleBehavior (Variables|ToyProperties|GetCycleBehavior)
    (:NewEnumerator0
      (Variables|Transient|SetGrantItemIndex (Math|Integer|Min(Integer) (Variables|Transient|GetGrantItemIndex) (Utilities|Array|LastIndex (Variables|Transient|GetValidItemIndexes)))))
    (:NewEnumerator1
      (if (>= (Variables|Transient|GetGrantItemIndex) (Utilities|Array|Length (Variables|Transient|GetValidItemIndexes)))
        (Variables|Transient|SetGrantItemIndex 0)
        (else
          _))))
  (if (Math|Conversions|ToBoolean(Integer) (Variables|ToyProperties|GetGrantOnIndexChange))
    (CallFunction|GrantItem)))

(event Custom|CycleToNextItemFromEvent
  (params (Instigator :type Controller%20Object%20Reference))
  ;[0.6921 0 1] Triggers & Receivers
  (Variables|Transient|SetTriggerInstigator Instigator)
  (Variables|Transient|SetGrantItemIndex (+ (Variables|Transient|GetGrantItemIndex) 1))
  (Variables|Default|SetEGrantProcessType "NewEnumerator19")
  (switch Utilities|FlowControl|Switch|SwitchonE_Device_ItemGranter_CycleBehavior (Variables|ToyProperties|GetCycleBehavior)
    (:NewEnumerator0
      (Variables|Transient|SetGrantItemIndex (Math|Integer|Min(Integer) (Variables|Transient|GetGrantItemIndex) (Utilities|Array|LastIndex (Variables|Transient|GetValidItemIndexes)))))
    (:NewEnumerator1
      (if (>= (Variables|Transient|GetGrantItemIndex) (Utilities|Array|Length (Variables|Transient|GetValidItemIndexes)))
        (Variables|Transient|SetGrantItemIndex 0)
        (else
          _))))
  (if (Math|Conversions|ToBoolean(Integer) (Variables|ToyProperties|GetGrantOnIndexChange))
    (CallFunction|GrantItem)))

(event Custom|TimerGrantRandom
  ;[0.6921 0 1] Triggers & Receivers
  (Variables|Transient|SetGrantItemIndex (Math|Random|RandomInteger (Utilities|Array|Length (Variables|Transient|GetValidItemIndexes))))
  (Variables|Default|SetEGrantProcessType "NewEnumerator20")
  (if (Math|Conversions|ToBoolean(Integer) (Variables|ToyProperties|GetGrantOnIndexChange))
    (CallFunction|GrantItem)))

(event Custom|CycleToRandomItemFromEvent
  (params (Instigator :type Controller%20Object%20Reference))
  ;[0.6921 0 1] Triggers & Receivers
  (Variables|Transient|SetTriggerInstigator Instigator)
  (Variables|Transient|SetGrantItemIndex (Math|Random|RandomInteger (Utilities|Array|Length (Variables|Transient|GetValidItemIndexes))))
  (Variables|Default|SetEGrantProcessType "NewEnumerator20")
  (if (Math|Conversions|ToBoolean(Integer) (Variables|ToyProperties|GetGrantOnIndexChange))
    (CallFunction|GrantItem)))

(event EventBeginPlay
  ;[0 1 0.1119 0.5] Initialization
  (|CallParentFunction)
  (Initialization|InitializeComponent (Variables|Default|GetCreative_ItemPreview_Component) (Variables|Components|GetPreviewMeshScene) (Variables|Components|GetPreviewStaticMesh) (Variables|Components|GetPreviewSkeletalMesh) (Variables|Components|GetHolo%20Loop%20Cue) (Variables|Default|GetPickupItemList) (Variables|Default|GetItemVariantNotifier) :!class /Game/Creative/Devices/Common/Components/Creative_ItemPreview_Component.Creative_ItemPreview_Component_C)
  (Networking|SwitchHasAuthority
    (:Authority
      (CallFunction|RegisterFortPersistenceComponent))))

(event OnItemListUpdated(PickupItemList)
  (CallFunction|UpdateSpawnerFX)
  (CallFunction|Resave))

(event Custom|RestockItemsFromEvent
  (params (Instigator :type Controller%20Object%20Reference))
  ;[0.6921 0 1] Triggers & Receivers
  (CallFunction|Restock))

(event Custom|StartGrantItemTimer
  ;[1 1 1] Grant Item on Timer
  (bind _returnvalue (Utilities|Time|SetTimerbyFunctionName self "GrantItemOnTimer" (Variables|ToyProperties|GetGrantItemTimer) true))
  (Variables|Transient|SetGrantItemTimerHandle _returnvalue))

(event Custom|ClearGrantItemTimer
  ;[1 1 1] Grant Item on Timer
  (Utilities|Time|ClearandInvalidateTimerbyHandle (Variables|Transient|GetGrantItemTimerHandle)))

(event EventOnAddedToMinigame
  (params (Minigame :type Fort%20Minigame%20Object%20Reference))
  ;[0 0.4933 1 0.5] Enables/Disables drop box based on island's play state (On Play Mode Changed is bound in parent)
  (|CallParentFunction Minigame)
  (CallFunction|OnPlayModeChanged Minigame (Minigame|IsinPlayMode Minigame :!class /Script/FortniteGame.FortMinigame)))

(event Custom|On%20Play%20Mode%20Changed
  (params (Minigame :type Fort%20Minigame%20Object%20Reference) (bIsInPlayMode :type Boolean))
  ;[0 0.4933 1 0.5] Enables/Disables drop box based on island's play state (On Play Mode Changed is bound in parent)
  (Collision|SetCollisionEnabled (Variables|Components|GetDropBox) (select bIsInPlayMode "NoCollision" "QueryOnly")))

(event Custom|ClearSaveDataForPlayerFromEvent
  (params (Instigator :type Controller%20Object%20Reference))
  ;[1 1 1] Persistence
  (if (and (Utilities|IsValid Instigator :!class /Script/Engine.KismetSystemLibrary) (CallFunction|IsinProgress))
    (bind _asfort_player_state (Utilities|Casting|CastToFortPlayerState (Variables|/Script/Engine.Controller|GetPlayerState Instigator)))
    (Persistence|ClearData (Variables|Default|GetFortCreativePersistenceComponentBase) (PlayerState|GetPlayerUniqueNetID _asfort_player_state) (Variables|Default|GetSave%20Path))))

(event Custom|RPC_MULTICAST%20Add%20Item
  (params (Item :type Fort%20Pickup%20Athena%20Object%20Reference))
  (Networking|CanExecuteCosmeticEvents
    (:True
      (Class|CreativeItemPreviewComponent|ItemAddedToQueueFX (Variables|Default|GetCreative_ItemPreview_Component) Item))))

(event Custom|UpdateSpawnerFX
  ;[1 1 1] Show or hide emissive elements based on spawning being enabled
  (Networking|CanExecuteCosmeticEvents
    (:True
      (Development|SetHiddeninGame (Variables|Components|GetLightRays) (not (Math|Conversions|ToBoolean(Integer) (ItemData|GetNumItems (Variables|Default|GetPickupItemList)))) true))))

(fn CalculateItemNumber
  (bind _processinggrantrequestspawnparams (Variables|Transient|GetProcessingGrantRequestSpawnParams))
  (bind (_pickupquantity _itemvariantguid) (Utilities|Struct|BreakMinigameSpawnerSpawnParams _processinggrantrequestspawnparams))
  (bind (_pickupinstigatorhandle _itemvariantguid_2) (Utilities|Struct|BreakMinigameSpawnerSpawnParams (Variables|Transient|GetProcessingGrantRequestSpawnParams)))
  (bind (_processinggrantrequest_controllers_38_822ad16d47dcab66ef0d7c9ce6898856 _processinggrantrequest_grantitemindex_21_b2c0fe6d40d1128d9ea9c192f324d978 _processinggrantrequest_countoverride_41_d8a66e564fea2a9fded634a7eddc274c _processinggrantrequest_executeitem_8_e7c22d144c7730e28a63608db73c7276 _processinggrantrequest_removeitem_20_8b360fe344295a37c03d84bfac65befd _processinggrantrequest_grantbehavior_32_af1cd62340c36f7d698dd59f7c7506f9 _processinggrantrequest_grantcondition_34_0e6824e546464c43b5988bad6d7f9c9e) (Utilities|Struct|BreakSCreativeItemGranterGrantRequest (Variables|GetProcessingGrantRequest_Controllers_38_822AD16D47DCAB66EF0D7C9CE6898856)))
  ;[0 1 0.8098 0.25] Grants the number of items in the device * the number of times it should have activated while the loaded player was away.
  ; If the device is set to 'only grant if owned' will only grant the number of items required to top up the player's inventory to the number specified.
  (LocalVariables|SetDesiredNum (* _pickupquantity (Math|Integer|SelectInt (Variables|ToyProperties|GetRemoveItemOnGrant) (Variables|Transient|GetItemGrantMultiplier) (Math|Conversions|ToBoolean(Integer) (Variables|ToyProperties|GetRemoveItemOnGrant)))))
  (if (Utilities|Enum|Equal(Enum) _processinggrantrequest_grantcondition_34_0e6824e546464c43b5988bad6d7f9c9e "NewEnumerator20")
    (LocalVariables|SetDesiredNum (Math|Integer|Max(Integer) (- (LocalVariables|GetDesiredNum) (Inventory|GetItemQuantityonInventoryOwner (Variables|Transient|GetProcessingGrantRequestInventoryOwner) (Variables|Transient|GetProcessingGrantRequestWorldItemDefinition) _itemvariantguid))))
    (else
      _))
  (if (or (Variables|UserOptions|GetbDropItemsAtPlayerLocationIfInventoryIsFull) (Variables|UserOptions|GetbDropItemsAtPlayerLocation))
    (return (LocalVariables|GetDesiredNum))
    (else
      (return (- (LocalVariables|GetDesiredNum) (Inventory|GetOverflowCountIfItemWasAddedToInventory (Variables|Transient|GetProcessingGrantRequestInventoryOwner) (Creative|ItemData|CreateItemVariantEntry (Variables|Transient|GetProcessingGrantRequestWorldItemDefinition) (LocalVariables|GetDesiredNum) -1 _itemvariantguid_2 _pickupinstigatorhandle)))))))

(fn CalculateItemsToGrantOverTime
  (bind _persistencedataitemgranter (Variables|Default|GetPersistenceDataItemGranter))
  (bind _processinggrantrequestcontroller (Variables|Transient|GetProcessingGrantRequestController))
  (Utilities|Array|AddUnique (Variables|Transient|GetPlayersGrantedItemsThisSession) _processinggrantrequestcontroller)
  (bind _asfort_player_state (Utilities|Casting|CastToFortPlayerState (Variables|/Script/Engine.Controller|GetPlayerState _processinggrantrequestcontroller)))
  (LocalVariables|SetCurrentGrantedPlayerState _asfort_player_state)
  (Persistence|LoadData (Variables|Default|GetFortCreativePersistenceComponentBase) (PlayerState|GetPlayerUniqueNetID _asfort_player_state) (Variables|Default|GetSave%20Path) (Variables|Default|GetPersistenceDataItemGranter))
  (if (not (Variables|/Game/Creative/Devices/ItemGranter/Persistence/PersistenceData_ItemGranter.PersistenceData_ItemGranter_C|GetPlayerBeingTracked (Variables|Default|GetPersistenceDataItemGranter)))
    (Variables|Transient|SetItemGrantMultiplier 1)
    (else
      (Variables|Transient|SetItemGrantMultiplier (Math|Float|Truncate (/ (Math|Timespan|GetTotalSeconds (Math|DateTime|DateTime-DateTime (Math|DateTime|UTCNow) (Variables|/Game/Creative/Devices/ItemGranter/Persistence/PersistenceData_ItemGranter.PersistenceData_ItemGranter_C|GetLastTimeGrantedToPlayer (Variables|Default|GetPersistenceDataItemGranter)))) (Variables|ToyProperties|GetGrantItemTimer))))))
  (Variables|/Game/Creative/Devices/ItemGranter/Persistence/PersistenceData_ItemGranter.PersistenceData_ItemGranter_C|SetLastTimeGrantedToPlayer (Math|DateTime|UTCNow) _persistencedataitemgranter)
  (Variables|/Game/Creative/Devices/ItemGranter/Persistence/PersistenceData_ItemGranter.PersistenceData_ItemGranter_C|SetPlayerBeingTracked true _persistencedataitemgranter)
  (Persistence|SaveData (Variables|Default|GetFortCreativePersistenceComponentBase) (Variables|Default|GetPersistenceDataItemGranter) (PlayerState|GetPlayerUniqueNetID (LocalVariables|GetCurrentGrantedPlayerState)) (Variables|Default|GetSave%20Path) false))

(fn CanPlayerAutoSave
  (params (FortPlayerState :type Fort%20Player%20State%20Object%20Reference))
  (return false))

(fn CreateGrantRequestForGroup
  (params (InTeamIndex :type Byte) (bIgnoreTeamIndex :type Boolean))
  ;[1 0.6347 0.1039] Create a grant request for all players in the minigame that are on the correct team
  (bind _fortminigameprogress (Variables|Default|GetFortMinigameProgress)
    (:then
      (bind _currentminigame (Variables|/Script/FortniteGame.FortMinigameLogicComponent|GetCurrentMinigame _fortminigameprogress)
        (:then
          (for _array_element (Minigame|GetParticipatingPlayers _currentminigame)
            (Utilities|IsValid (Fortnite|GetControllerfromActor _array_element)
              (:Is%20Valid
                (if (or (Variables|GetbIgnoreTeamIndex) (Math|Byte|Equal(Byte) (Team|GetActorPvPTeamIndex (Fortnite|GetControllerfromActor _array_element)) (Variables|GetInTeamIndex)))
                  (Utilities|Array|AddUnique (LocalVariables|GetParticipatingControllers) (Fortnite|GetControllerfromActor _array_element))))))
          (CallFunction|CreateGrantRequestforPlayers (LocalVariables|GetParticipatingControllers)))
        (:else)))
    (:else)))

(fn CreateGrantRequestForPlayers
  (params (Player%20Controllers :type Array%20of%20Controller%20Object%20References))
  (bind (_best_valid_item_index _isvalid) (CallFunction|GetBestValidItemIndex (Variables|Transient|GetGrantItemIndex)))
  (Utilities|FlowControl|Sequence
    (:then_0
      (switch Utilities|FlowControl|Switch|SwitchonE_Device_ItemGranter_GrantType (Variables|ToyProperties|GetGrantType)
        (:NewEnumerator0
          (Utilities|Array|Add (select (Variables|Transient|GetbHasCalledOnMinigameStarted) (Variables|Transient|GetGrantRequestQueue) (Variables|Transient|GetPendingGrantRequestQueue)) (Utilities|Struct|MakeSCreativeItemGranterGrantRequest (Variables|GetPlayer%20Controllers) (Variables|Transient|GetGrantItemIndex) 0 (and _isvalid (== _best_valid_item_index (Math|Integer|Min(Integer) (Variables|ToyProperties|GetEquipItemOnGrant) (Utilities|Array|LastIndex (Variables|Transient|GetValidItemIndexes))))) (Math|Conversions|ToBoolean(Integer) (Variables|ToyProperties|GetRemoveItemOnGrant)) (Variables|ToyProperties|GetGrantBehavior) (Variables|ToyProperties|GetGrantCondition) true)))
        (:NewEnumerator1
          (for _index (range (+ (Utilities|Array|LastIndex (Variables|Transient|GetValidItemIndexes)) 1))
            (Utilities|Array|Add (select (Variables|Transient|GetbHasCalledOnMinigameStarted) (Variables|Transient|GetGrantRequestQueue) (Variables|Transient|GetPendingGrantRequestQueue)) (Utilities|Struct|MakeSCreativeItemGranterGrantRequest (Variables|GetPlayer%20Controllers) _index 0 (== (Math|Integer|Min(Integer) (Variables|ToyProperties|GetEquipItemOnGrant) (Utilities|Array|LastIndex (Variables|Transient|GetValidItemIndexes))) _index) (Math|Conversions|ToBoolean(Integer) (Variables|ToyProperties|GetRemoveItemOnGrant)) (select (== _index 0) (Variables|ToyProperties|GetGrantBehavior) "NewEnumerator3") (select (== _index 0) (Variables|ToyProperties|GetGrantCondition) (Utilities|Select :Index (Variables|ToyProperties|GetGrantCondition) :NewEnumerator0 "NewEnumerator0" :NewEnumerator1 "NewEnumerator19" :NewEnumerator2 "NewEnumerator2" :NewEnumerator19 "NewEnumerator19" :NewEnumerator20 "NewEnumerator20")) (== _index (Utilities|Array|LastIndex (Variables|Transient|GetValidItemIndexes)))))))))
    (:then_1
      (CallFunction|ProcessGrantRequestQueue))))

(fn GetBestValidItemIndex
  (params (In%20Grant%20Item%20Index :type Integer))
  (for _index (range In%20Grant%20Item%20Index (+ (Utilities|Array|LastIndex (Variables|Transient|GetValidItemIndexes)) 1))
    (if (!= (Utilities|Array|Get(aref) (Variables|Transient|GetValidItemIndexes) _index) -1)
      (return _index true)))
  (return -1 false))

(fn GetObjectToAutoSaveForPlayer
  (params (FortPlayerState :type Fort%20Player%20State%20Object%20Reference))
  (return 0))

(fn GetPersistenceRequiredState
  (return (== (Variables|ToyProperties|GetGrantWhileOffline) 1)))

(fn GetVisibilityComponentsWithCollision
  (return (Utilities|Map|MakeMap (Variables|Components|GetLightRays) "NoCollision" (Variables|Debug|GetStaticMeshComponent) "QueryAndPhysics" (Variables|Components|GetPointLight) "NoCollision" (Variables|Components|GetPreviewStaticMesh) "NoCollision" (Variables|Components|GetPreviewSkeletalMesh) "NoCollision" (Variables|Components|GetHolo%20Loop%20Cue) "NoCollision")))

(fn GrantItemOnTimer
  (bind _persistencedataitemgranter (Variables|Default|GetPersistenceDataItemGranter))
  (if (CallFunction|IsinProgress)
    (if (Variables|Transient|GetbEnabled)
      (switch Utilities|FlowControl|Switch|SwitchonE_Device_ItemGranterGrantProcessType (Variables|Default|GetEGrantProcessType)
        (:NewEnumerator0
          (CallFunction|TimerGrantItem))
        (:NewEnumerator1
          (CallFunction|TimerGrantPrevious))
        (:NewEnumerator19
          (CallFunction|TimerGrantNext))
        (:NewEnumerator20
          (CallFunction|TimerGrantRandom)))
      (else
        (CallFunction|ClearGrantItemTimer)
        (if (== (Variables|ToyProperties|GetGrantWhileOffline) 1)
          (for _array_element (Variables|Transient|GetPlayersGrantedItemsThisSession)
            (bind _asfort_player_state (Utilities|Casting|CastToFortPlayerState (Variables|/Script/Engine.Controller|GetPlayerState _array_element)))
            (if (Minigame|IsPlayerinOurVolume (Minigame|GetMinigame self :!class /Script/FortniteGame.MinigameBlueprintLibrary) _asfort_player_state)
              (Variables|/Game/Creative/Devices/ItemGranter/Persistence/PersistenceData_ItemGranter.PersistenceData_ItemGranter_C|SetPlayerBeingTracked false _persistencedataitemgranter)
              (Persistence|SaveData (Variables|Default|GetFortCreativePersistenceComponentBase) _persistencedataitemgranter (PlayerState|GetPlayerUniqueNetID _asfort_player_state) (Variables|Default|GetSave%20Path) false))))))))

(fn IsInProgress
  (bind _fortminigameprogress (Variables|Default|GetFortMinigameProgress)
    (:then
      (bind _currentminigame (Variables|/Script/FortniteGame.FortMinigameLogicComponent|GetCurrentMinigame _fortminigameprogress)
        (:then
          (return (Minigame|IsinProgress _currentminigame)))
        (:else)))
    (:else)))

(fn IsPlayerAllowedToEquipProcessingItem
  (bind _returnvalue (Fortnite|Athena|GameState|GetGameStateAthena))
  ;[1 1 1] If not a trap, this function currently always returns true
  (if (Utilities|Enum|Equal(Enum) (Item|GetItemType (Variables|Transient|GetProcessingGrantRequestWorldItemDefinition)) "Trap")
    (if (Utilities|IsValid _returnvalue :!class /Script/Engine.KismetSystemLibrary)
      (return (not (Class|FortGameStateAthena|ShouldBlockTrapPlacement _returnvalue (Pawn|GetControlledPawn (Variables|Transient|GetProcessingGrantRequestController)))))
      (else
        (return true)))
    (else
      (return true))))

(fn IsVisibleDuringMinigame
  (return false))

(fn MaybeGiveAmmoToInventoryOwner
  (params (InventoryOwner :type Fort%20Inventory%20Owner%20Interface%20Interface) (WorldItemDefinition :type Fort%20World%20Item%20Definition%20Object%20Reference))
  (if (Math|Conversions|ToBoolean(Integer) (Variables|ToyProperties|GetGiveExtraAmmo))
    (HelperFunctions|GiveAmmotoInventoryOwner InventoryOwner WorldItemDefinition -1 (Variables|Default|GetExtra%20Ammo%20Amounts%20Map) (Variables|Default|GetDefaultExtraAmmoAmount))))

(fn RegisterFortPersistenceComponent
  (Persistence|InitializePersistenceComponent (Variables|Default|GetFortCreativePersistenceComponentBase) (Variables|Default|GetFortMinigameProgress) true))

(fn Restock
  (Utilities|Array|Clear (Variables|Transient|GetValidItemIndexes))
  (for _index (range (ItemData|GetNumItems (Variables|Default|GetPickupItemList)))
    (Utilities|Array|AddUnique (Variables|Transient|GetValidItemIndexes) _index)))

(fn SendWhenItemGrantedTriggerMessage
  (params (Trigger%20Instigator :type Controller%20Object%20Reference))
  (GameplayEvent|BroadcastEvent (Variables|UserOptions|Events|GetOnItemGranted) 0 (Utilities|Struct|MakeGameplayEventItemGranted Trigger%20Instigator)))

(fn ShouldSpawnedItemDropToGround
  (if (Variables|UserOptions|GetbDropItemsAtPlayerLocation)
    (return true)
    (elif (Utilities|IsA (Variables|Transient|GetProcessingGrantRequestWorldItemDefinition) "/Script/FortniteGame.FortWeaponItemDefinition")
      (if (and (Variables|UserOptions|GetbDropItemsAtPlayerLocationIfInventoryIsFull) (== (Inventory|GetInventorySpace (Variables|Transient|GetProcessingGrantRequestInventoryOwner)) 0))
        (return true)
        (else
          (return false)))
      (else
        (return false)))))

(fn UnRegisterFortPersistenceComponent
  (params (FortMinigame :type Fort%20Minigame%20Object%20Reference))
  (bind _volume (Variables|/Script/FortniteGame.FortMinigame|GetVolume FortMinigame))
  (bind _returnvalue (Actor|GetComponentbyClass _volume "/Script/FortniteGame.FortVolumePersistenceOptions"))
  (Utilities|IsValid _returnvalue
    (:Is%20Valid
      (Class|FortVolumePersistenceOptions|UnregistryFortCreativePersistenceComponent _returnvalue 0))))

(fn ConstructionScript
  (|CallParentFunction))
